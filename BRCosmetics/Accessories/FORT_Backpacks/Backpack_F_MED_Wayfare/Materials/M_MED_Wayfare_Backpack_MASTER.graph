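# Master character material for the Wayfare backpack, adding an animated emissive screen to the standard character PreFX, skin FX and PostFX chain.
# The emissive texture uses cropped, center-scaled UVs, with an optional faint bump-offset copy for fake depth, driven by inverted MaterialCurveA.
# The tinted result blends with a row-staggered Jumbotron CRT pattern before intensity scaling.
# Named reroutes carry crop UVs, blink, time-based flicker and the final emissive.
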Material /BRCosmetics/Accessories/FORT_Backpacks/Backpack_F_MED_Wayfare/Materials/M_MED_Wayfare_Backpack_MASTER
graph {
 properties {
  bAutomaticallySetUsageInEditor=false
  bUseMaterialAttributes=true
  bUsedWithMeshDeformer=true
  bUsedWithMorphTargets=true
  bUsedWithNanite=true
  bUsedWithSkeletalMesh=true
  pixelDepthOffsetMode=PDOM_Legacy
  refractionMethod=RM_IndexOfRefraction
  shadingModel=MSM_FromMaterialExpression
  subsurfaceProfile=&SS_HeroSkin_02
  usedShadingModels=Default Lit | Subsurface | Clear Coat | Subsurface Profile | Cloth | Eye
 }
 nodes [
  n1 = MaterialFunctionCall() {materialFunction=&BaseFN_CharacterFunction_PreFX}
  n2 = GetMaterialAttributes(None=MaterialFunctionCall(
    PreskinnedPosition=@n1.Pre skinned Local Position
    Gmap=@n1.Pre skinned Local Position
    MaterialAttributes=@n1.Pre skinned Local Position
    AO=@n1.Pre skinned Local Position
   ) {materialFunction=&CharacterSkinFX}.Result) {attributeGetTypes [#1]}
  n3 = NamedRerouteUsage()
  n4 = MaterialFunctionCall(
   AO=@n1.AO
   Distance Blend=@n1.AO
   Fresnel=@n1.AO
   Gmap=@n1.AO
   Pre Skinned Local Position=@n1.AO
   Material=SetMaterialAttributes(MaterialAttributes=@n2.MaterialAttributes, Emissive Color=Add(A=@n2.Emissive Color, B=@n3)) {attributeSetTypes [#1]}
  ) {materialFunction=&BaseFN_CharacterFunction_PostFX}
  MaterialFunctionCall(In=@n4.Result) {materialFunction=&CINE_MF_CharPostFX}
  n5 = TextureSampleParameter2D(UVs=MaterialFunctionCall(UVs=NamedRerouteUsage(), Texture Scale=NamedRerouteUsage()) {materialFunction=&ScaleUVsByCenter}.UVs) {
   group=Wayfare Emissive
   parameterName=Emissive
   samplerSource=SSM_Clamp_WorldGroupSettings
   texture=/BRCosmetics/Accessories/FORT_Backpacks/Backpack_M_MED_Wayfare/Textures/T_M_MED_Wayfare_Backpack_E
  }
  n6 = TextureSampleParameter2D(UVs=MaterialFunctionCall(UVs=NamedRerouteUsage(), Texture Scale=NamedRerouteUsage()) {materialFunction=&ScaleUVsByCenter}.UVs) {
   group=Wayfare Emissive
   parameterName=Emissive
   samplerSource=SSM_Clamp_WorldGroupSettings
   texture=&T_M_MED_Wayfare_Backpack_E
  }
  n7 = Multiply(A=@n6.RGB, B=ScalarParameter() {defaultValue=0.025, group=Wayfare Emissive, parameterName=Offset Opacity})
  n8 = VectorParameter() {group=Wayfare Emissive, parameterName=Emissive Color, sortPriority=15, defaultValue=<rgba 0.1559 0.4397 1 1>}
  n9 = Multiply(
   A=StaticSwitchParameter(True=Add(A=@n5.RGB, B=@n7), False=@n5.RGB) {defaultValue=true, group=Wayfare Emissive, parameterName=useDepthOffset}
   B=@n8.RGB
  )
  n10 = Multiply(A=NamedRerouteUsage(), B=ScalarParameter() {defaultValue=10, parameterName=CRTScaleUV})
  n11 = ComponentMask(None=@n10) {g=true}
  n12 = AppendVector(A=Sign(None=Sine(None=@n11) {period=2}), B=Constant())
  n13 = TextureSample(UVs=Add(A=@n10, B=Multiply(A=@n12) {constB=0.25})) {samplerType=SAMPLERTYPE_LinearColor, texture=/Game/Accessories/FORT_Backpacks/Backpack_M_MED_Jumbotron_S20/Textures/T_Jumbotron_CRT}
  n14 = ScalarParameter() {defaultValue=0.8, parameterName=CRT_ColorScale}
  n15 = LinearInterpolate(A=@n9, B=Multiply(A=@n9, B=Multiply(A=@n13.G, B=@n14)), Alpha=ScalarParameter() {defaultValue=0.4, parameterName=CRT_Amount})
  n16 = NamedRerouteUsage()
  NamedRerouteDeclaration(None=Multiply(A=Multiply(A=@n15, B=@n16), B=ScalarParameter() {defaultValue=5, group=Wayfare Emissive, parameterName=Emissive Intensity})) {name=Emissive_Out, nodeColor=<rgba 0 1 0.7422 1>}
  n17 = TextureObjectParameter() {
   group=Wayfare Emissive
   parameterName=Emissive
   texture=&T_M_MED_Wayfare_Backpack_E
  }
  n18 = VectorParameter() {group=Wayfare Emissive, parameterName=EmissiveUVs_RG_UpperLeftCorner_BA_LowerRightCorner, defaultValue=<rgba 0.521 0.1356 0.624 0.0392>}
  n19 = ComponentMask(None=@n18.RGB) {g=true, r=true}
  n20 = AppendVector(A=@n18.B, B=@n18.B)
  NamedRerouteDeclaration(None=MaterialFunctionCall!(UVs, TextureIn=@n17, UpperLeftCorner=@n19, LowerRightCorner=@n20, ExportFloat4) {materialFunction=/Engine/Functions/Engine_MaterialFunctions01/Texturing/TextureCropping}.Crop UVs) {name=CropUV, nodeColor=<rgba 1 0 0.5156 1>}
  n21 = NamedRerouteDeclaration(None=OneMinus(None=ScalarParameter() {group=Wayfare Emissive, parameterName=MaterialCurveA})) {name=Blink, nodeColor=<rgba 1 0 0.4453 1>}
  n22 = Constant() {r=0.5}
  n23 = Multiply(A=ScalarParameter() {defaultValue=1, group=Wayfare Emissive, parameterName=Depth Offset}, B=Add(A=Multiply(A=@n21, B=@n22), B=@n22))
  NamedRerouteDeclaration(None=MaterialFunctionCall(UVs=BumpOffset(Height=@n23), TextureIn=@n17, UpperLeftCorner=@n19, LowerRightCorner=@n20) {materialFunction=&TextureCropping}.Crop UVs) {name=Second UVs, nodeColor=<rgba 0 1 0.6719 1>}
  n24 = Sine(None=Multiply(A=ScalarParameter() {defaultValue=1, group=Wayfare Emissive, parameterName=Flicker Speed}, B=Time()))
  NamedRerouteDeclaration(None=Clamp(None=Frac(None=@n24), Min=ScalarParameter() {defaultValue=0.4, group=Wayfare Emissive, parameterName=Flicker Opacity})) {name=Flicker, nodeColor=<rgba 0.7344 0 1 1>}
  ScalarParameter() {parameterName=MaterialCurveB}
 ]
}
