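# Animation Blueprint for the Wayfare backpack cosmetic, animating the pack in response to its wearer's actions.
# A state machine covers locomotion, targeting, melee, emotes, jumping, sliding, skydiving and DBNO/revive, with randomized idles and direction-picked run starts.
# Bone modifiers then scale the pack and add hover and carrier-velocity rotation, cached as PreSpringPose.
# The final pose blends between an additive carrier offset and a world-space target position by a smoothed alpha.
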AnimBlueprint /BRCosmetics/Accessories/FORT_Backpacks/Backpack_F_MED_Wayfare/Meshes/F_MED_Wayfare_Pack_AnimBP
target_skeleton=/Game/Accessories/FORT_Backpacks/Backpack_F_MED_Wayfare/Meshes/F_MED_Wayfare_Pack_Skeleton
graphs [
 AnimGraph {nodes [
   n1 = "Transform (Modify) Bone - Bone: pack"(
    ComponentPose=Local To Component(LocalPose=Use cached pose 'PreSpringPose'())
    Translation=Get BackpackTargetVector() -> (BackpackTargetVector:Vector)
    Rotation="0,0,0"
   ) {translationMode=BMM_Replace, translationSpace=BCS_WorldSpace, alphaBoolBlend {blendOption=Linear}, boneToModify {boneName=pack}}
   n2 = "Transform (Modify) Bone - Bone: pack"(
    ComponentPose=Local To Component(LocalPose=Use cached pose 'PreSpringPose'())
    Translation=Get CarrierOffsetTotal() -> (CarrierOffsetTotal:Vector)
    Rotation="0,0,0"
   ) {translationMode=BMM_Additive, alphaBoolBlend {blendOption=Linear}, boneToModify {boneName=pack}}
   Output Pose(Result=Two Way Blend(
     A=Component To Local(ComponentPose=@n2)
     B=Component To Local(ComponentPose=@n1)
     Alpha=Get BackpackTargetVectorAlphaSmooth() -> (BackpackTargetVectorAlphaSmooth:float)
    ) {blendNode {alphaBoolBlend {blendOption=Linear}}}) {layerGroup=None}
   AnimGraphNode_StateMachine_0 = New State Machine() {is=title AnimGraphNode_StateMachine}
   n3 = Local To Component(LocalPose=@AnimGraphNode_StateMachine_0)
   n4 = Get HoverRotationOffsetTotal() -> (HoverRotationOffsetTotal:Rotator)
   n5 = "Transform (Modify) Bone - Bone: pack"(
    ComponentPose="Transform (Modify) Bone - Bone: offset_leaf"(
     ComponentPose="Transform (Modify) Bone - Bone: offset_leaf"(ComponentPose=@n3, Alpha=1, Translation="0,0,0", Rotation=@n4, Scale="1,1,1") {rotationMode=BMM_Additive, alphaBoolBlend {blendOption=Linear}, boneToModify {boneName=offset_leaf}}
     Alpha=Get CarrierOffsetRotationVelocityAlpha() -> (CarrierOffsetRotationVelocityAlpha:float)
     Translation="0,0,0"
     Rotation=Get CarrierOffsetRotationVelocity() -> (CarrierOffsetRotationVelocity:Rotator)
     Scale="1,1,1"
    ) {rotationMode=BMM_Additive, alphaBoolBlend {blendOption=Linear}, boneToModify {boneName=offset_leaf}}
    Translation="0,0,0"
    Rotation=Get RotationYawDefinedHoverOffset() -> (RotationYawDefinedHoverOffset:Rotator)
    Scale="0.6,0.6,0.6"
   ) {scaleMode=BMM_Replace, alphaBoolBlend {blendOption=Linear}, boneToModify {boneName=pack}, scale=<0.6 0.6 0.6>}
   Save cached pose 'PreSpringPose'(Pose=Component To Local(ComponentPose=@n5)) {cacheName=PreSpringPose}
  ]}
 New State Machine {
  entry=GroundLocomotion
  path=AnimGraph.AnimGraphNode_StateMachine_0.New State Machine
  ids=id
  nodes [
   AnimStateNode_0 = GroundLocomotion() {
    is=title AnimStateNode
    stateEntered {guid=#1}
    stateFullyBlended {guid=#2}
    stateLeft {guid=#3}
   }
   AnimStateNode_1 = Targeting() {is=title AnimStateNode}
   AnimStateNode_10 = Melee() {is=title AnimStateNode}
   AnimStateNode_2 = DBNO() {
    is=title AnimStateNode
    stateEntered {guid=#4}
    stateFullyBlended {guid=#5}
    stateLeft {guid=#6}
   }
   AnimStateNode_3 = SkyDiveMode() {is=title AnimStateNode}
   AnimStateNode_4 = Emote() {is=title AnimStateNode}
   AnimStateNode_5 = DBNO_End() {is=title AnimStateNode, stateLeft {guid=#7, notifyName=DBNO_Exited}}
   AnimStateNode_7 = Jump() {is=title AnimStateNode}
   AnimStateNode_9 = Sliding() {is=title AnimStateNode}
   AnimStateTransitionNode_0 = AnimStateTransitionNode()
   AnimStateTransitionNode_1 = AnimStateTransitionNode()
   AnimStateTransitionNode_10 = AnimStateTransitionNode()
   AnimStateTransitionNode_11 = AnimStateTransitionNode() {bAutomaticRuleBasedOnSequencePlayerInState=true}
   AnimStateTransitionNode_15 = AnimStateTransitionNode()
   AnimStateTransitionNode_16 = AnimStateTransitionNode()
   AnimStateTransitionNode_17 = AnimStateTransitionNode()
   AnimStateTransitionNode_18 = AnimStateTransitionNode()
   AnimStateTransitionNode_2 = AnimStateTransitionNode()
   AnimStateTransitionNode_3 = AnimStateTransitionNode() {crossfadeDuration=0.15}
   AnimStateTransitionNode_4 = AnimStateTransitionNode() {crossfadeDuration=0.1, logicType=TLT_Inertialization}
   AnimStateTransitionNode_5 = AnimStateTransitionNode()
   AnimStateTransitionNode_6 = AnimStateTransitionNode()
   AnimStateTransitionNode_7 = AnimStateTransitionNode()
   AnimStateTransitionNode_8 = AnimStateTransitionNode()
   AnimStateTransitionNode_9 = AnimStateTransitionNode() {crossfadeDuration=0.5}
  ]
  links [
   {
    dst=DBNO_End
    rule=AnimGraph.AnimGraphNode_StateMachine_0.New State Machine.AnimStateTransitionNode_4.Transition
    src=DBNO
    via=AnimStateTransitionNode_4
   }
   {
    dst=GroundLocomotion
    rule=AnimGraph.AnimGraphNode_StateMachine_0.New State Machine.AnimStateTransitionNode_3.AnimationTransitionGraph_2
    src=DBNO
    via=AnimStateTransitionNode_3
   }
   {
    dst=GroundLocomotion
    rule=AnimGraph.AnimGraphNode_StateMachine_0.New State Machine.AnimStateTransitionNode_11.Transition
    src=DBNO_End
    via=AnimStateTransitionNode_11
   }
   {
    dst=GroundLocomotion
    rule=AnimGraph.AnimGraphNode_StateMachine_0.New State Machine.AnimStateTransitionNode_8.Transition
    src=Emote
    via=AnimStateTransitionNode_8
   }
   {
    dst=DBNO
    rule=AnimGraph.AnimGraphNode_StateMachine_0.New State Machine.AnimStateTransitionNode_10.AnimationTransitionGraph_2
    src=GroundLocomotion
    via=AnimStateTransitionNode_10
   }
   {
    dst=Emote
    rule=AnimGraph.AnimGraphNode_StateMachine_0.New State Machine.AnimStateTransitionNode_2.Transition
    src=GroundLocomotion
    via=AnimStateTransitionNode_2
   }
   {
    dst=Jump
    rule=AnimGraph.AnimGraphNode_StateMachine_0.New State Machine.AnimStateTransitionNode_6.Transition
    src=GroundLocomotion
    via=AnimStateTransitionNode_6
   }
   {
    dst=Melee
    rule=AnimGraph.AnimGraphNode_StateMachine_0.New State Machine.AnimStateTransitionNode_0.Transition
    src=GroundLocomotion
    via=AnimStateTransitionNode_0
   }
   {
    dst=SkyDiveMode
    rule=AnimGraph.AnimGraphNode_StateMachine_0.New State Machine.AnimStateTransitionNode_5.Transition
    src=GroundLocomotion
    via=AnimStateTransitionNode_5
   }
   {
    dst=Sliding
    rule=AnimGraph.AnimGraphNode_StateMachine_0.New State Machine.AnimStateTransitionNode_17.Transition
    src=GroundLocomotion
    via=AnimStateTransitionNode_17
   }
   {
    dst=Targeting
    rule=AnimGraph.AnimGraphNode_StateMachine_0.New State Machine.AnimStateTransitionNode_15.Transition
    src=GroundLocomotion
    via=AnimStateTransitionNode_15
   }
   {
    dst=GroundLocomotion
    rule=AnimGraph.AnimGraphNode_StateMachine_0.New State Machine.AnimStateTransitionNode_9.Transition
    src=Jump
    via=AnimStateTransitionNode_9
   }
   {
    dst=GroundLocomotion
    rule=AnimGraph.AnimGraphNode_StateMachine_0.New State Machine.AnimStateTransitionNode_1.Transition
    src=Melee
    via=AnimStateTransitionNode_1
   }
   {
    dst=GroundLocomotion
    rule=AnimGraph.AnimGraphNode_StateMachine_0.New State Machine.AnimStateTransitionNode_7.Transition
    src=SkyDiveMode
    via=AnimStateTransitionNode_7
   }
   {
    dst=GroundLocomotion
    rule=AnimGraph.AnimGraphNode_StateMachine_0.New State Machine.AnimStateTransitionNode_18.Transition
    src=Sliding
    via=AnimStateTransitionNode_18
   }
   {
    dst=GroundLocomotion
    rule=AnimGraph.AnimGraphNode_StateMachine_0.New State Machine.AnimStateTransitionNode_16.Transition
    src=Targeting
    via=AnimStateTransitionNode_16
   }
  ]
 }
 GroundLocomotion {
  path=AnimGraph.AnimGraphNode_StateMachine_0.New State Machine.AnimStateNode_0.GroundLocomotion
  nodes [
   Random Sequence Player!() -> (Pose:Pose Link Structure) {
    is=title AnimGraphNode_RandomPlayer
    entries [
     {
      chanceToPlay=0.6
      maxLoopCount=0
      maxPlayRate=1
      minLoopCount=0
      minPlayRate=1
      sequence=/BRCosmetics/Accessories/FORT_Backpacks/Backpack_F_MED_Wayfare/WayfareAnimations/Idle_F_MED_Wayfare_Pack
      blendIn {blendOption=HermiteCubic, blendTime=0.3, customCurve=null}
     }
     {
      chanceToPlay=0.15
      maxLoopCount=0
      maxPlayRate=1
      minLoopCount=0
      minPlayRate=0
      sequence=/BRCosmetics/Accessories/FORT_Backpacks/Backpack_F_MED_Wayfare/WayfareAnimations/IdleBreak_F_MED_Wayfare_Pack
      blendIn {blendOption=HermiteCubic, blendTime=0.2, customCurve=null}
     }
     {
      chanceToPlay=0.15
      maxLoopCount=0
      maxPlayRate=1
      minLoopCount=0
      minPlayRate=0
      sequence=/BRCosmetics/Accessories/FORT_Backpacks/Backpack_F_MED_Wayfare/WayfareAnimations/IdleBreakPeek_F_MED_Wayfare_Pack
      blendIn {blendOption=HermiteCubic, blendTime=0.2, customCurve=null}
     }
    ]
   }
   Sequence Player 'Idle_F_MED_Wayfare_Pack'() {sequence=&Idle_F_MED_Wayfare_Pack}
   AnimGraphNode_StateMachine_0 = GroundLocomotion() {is=title AnimGraphNode_StateMachine}
   Output Animation Pose(Result=@AnimGraphNode_StateMachine_0)
  ]
 }
 GroundLocomotion {
  entry=Idle_1
  path=AnimGraph.AnimGraphNode_StateMachine_0.New State Machine.AnimStateNode_0.GroundLocomotion.AnimGraphNode_StateMachine_0.GroundLocomotion
  ids=id
  nodes [
   AnimStateNode_0 = Idle_1() {
    is=title AnimStateNode
    stateEntered {guid=#1}
    stateFullyBlended {guid=#2}
    stateLeft {guid=#3}
   }
   AnimStateNode_1 = Run() {
    is=title AnimStateNode
    stateEntered {guid=#8}
    stateFullyBlended {guid=#9}
    stateLeft {guid=#10}
   }
   AnimStateNode_3 = Stop() {
    is=title AnimStateNode
    stateEntered {guid=#11}
    stateFullyBlended {guid=#12}
    stateLeft {guid=#13}
   }
   AnimStateNode_5 = Loop() {is=title AnimStateNode}
   AnimStateTransitionNode_0 = AnimStateTransitionNode() {crossfadeDuration=0.1}
   AnimStateTransitionNode_1 = AnimStateTransitionNode() {
    transitionEnd {guid=#14}
    transitionInterrupt {guid=#15}
    transitionStart {guid=#16}
   }
   AnimStateTransitionNode_11 = AnimStateTransitionNode() {crossfadeDuration=0}
   AnimStateTransitionNode_2 = AnimStateTransitionNode() {
    bAutomaticRuleBasedOnSequencePlayerInState=true
    transitionEnd {guid=#17}
    transitionInterrupt {guid=#18}
    transitionStart {guid=#19}
   }
   AnimStateTransitionNode_4 = AnimStateTransitionNode() {bAutomaticRuleBasedOnSequencePlayerInState=true, crossfadeDuration=0.3}
  ]
  links [
   {
    dst=Run
    rule=AnimGraph.AnimGraphNode_StateMachine_0.New State Machine.AnimStateNode_0.GroundLocomotion.AnimGraphNode_StateMachine_0.GroundLocomotion.AnimStateTransitionNode_0.AnimationTransitionGraph_4
    src=Idle_1
    via=AnimStateTransitionNode_0
   }
   {
    dst=Stop
    rule=AnimGraph.AnimGraphNode_StateMachine_0.New State Machine.AnimStateNode_0.GroundLocomotion.AnimGraphNode_StateMachine_0.GroundLocomotion.AnimStateTransitionNode_1.AnimationTransitionGraph_4
    src=Loop
    via=AnimStateTransitionNode_1
   }
   {
    dst=Loop
    rule=AnimGraph.AnimGraphNode_StateMachine_0.New State Machine.AnimStateNode_0.GroundLocomotion.AnimGraphNode_StateMachine_0.GroundLocomotion.AnimStateTransitionNode_2.AnimationTransitionGraph_4
    src=Run
    via=AnimStateTransitionNode_2
   }
   {
    dst=Idle_1
    rule=AnimGraph.AnimGraphNode_StateMachine_0.New State Machine.AnimStateNode_0.GroundLocomotion.AnimGraphNode_StateMachine_0.GroundLocomotion.AnimStateTransitionNode_11.AnimationTransitionGraph_4
    src=Stop
    via=AnimStateTransitionNode_11
   }
   {
    dst=Idle_1
    rule=AnimGraph.AnimGraphNode_StateMachine_0.New State Machine.AnimStateNode_0.GroundLocomotion.AnimGraphNode_StateMachine_0.GroundLocomotion.AnimStateTransitionNode_4.AnimationTransitionGraph_4
    src=Stop
    via=AnimStateTransitionNode_4
   }
  ]
 }
 Idle_1 {
  path=AnimGraph.AnimGraphNode_StateMachine_0.New State Machine.AnimStateNode_0.GroundLocomotion.AnimGraphNode_StateMachine_0.GroundLocomotion.AnimStateNode_0.Idle_1
  nodes [
   Sequence Player 'Idle_F_MED_Wayfare_Pack'() {sequence=&Idle_F_MED_Wayfare_Pack}
   n1 = Random Sequence Player() {entries [
     {
      chanceToPlay=0.7
      maxLoopCount=0
      maxPlayRate=1
      minLoopCount=0
      minPlayRate=1
      sequence=&Idle_F_MED_Wayfare_Pack
      blendIn {blendOption=HermiteCubic, blendTime=0.3, customCurve=null}
     }
     {
      chanceToPlay=0.15
      maxLoopCount=0
      maxPlayRate=1
      minLoopCount=0
      minPlayRate=1
      sequence=&IdleBreak_F_MED_Wayfare_Pack
      blendIn {blendOption=HermiteCubic, blendTime=0.2, customCurve=null}
     }
     {
      chanceToPlay=0.15
      maxLoopCount=0
      maxPlayRate=1
      minLoopCount=0
      minPlayRate=1
      sequence=&IdleBreakPeek_F_MED_Wayfare_Pack
      blendIn {blendOption=HermiteCubic, blendTime=0.2, customCurve=null}
     }
    ]}
   Output Animation Pose(Result=@n1)
  ]
 }
 Run {
  path=AnimGraph.AnimGraphNode_StateMachine_0.New State Machine.AnimStateNode_0.GroundLocomotion.AnimGraphNode_StateMachine_0.GroundLocomotion.AnimStateNode_1.Run
  nodes [
   n1 = Sequence Player 'FoldupStart_F_MED_Wayfare_Pack'() {
    bLoopAnimation=false
    sequence=/BRCosmetics/Accessories/FORT_Backpacks/Backpack_F_MED_Wayfare/WayfareAnimations/FoldupStart_F_MED_Wayfare_Pack
   }
   n2 = Sequence Player 'FoldupStart_F_MED_Wayfare_Pack'() {
    bLoopAnimation=false
    sequence=&FoldupStart_F_MED_Wayfare_Pack
   }
   Blend Poses by bool(BlendPose_0=@n1, BlendPose_1=@n2, BlendTime_0=0, BlendTime_1=0, bActiveValue=Get bIsCrouching()) {blendType=Linear, transitionType=Inertialization}
   AnimGraphNode_StateMachine_0 = New State Machine() {bAllowConduitEntryStates=true}
   Output Animation Pose(Result=@AnimGraphNode_StateMachine_0)
  ]
 }
 New State Machine {
  entry=Conduit
  path=AnimGraph.AnimGraphNode_StateMachine_0.New State Machine.AnimStateNode_0.GroundLocomotion.AnimGraphNode_StateMachine_0.GroundLocomotion.AnimStateNode_1.Run.AnimGraphNode_StateMachine_0.New State Machine
  ids=id
  nodes [
   AnimStateConduitNode_0 = Conduit()
   AnimStateNode_1 = North() {is=title AnimStateNode}
   AnimStateNode_2 = South() {is=title AnimStateNode}
   AnimStateNode_3 = East() {is=title AnimStateNode}
   AnimStateNode_4 = West() {is=title AnimStateNode}
   AnimStateTransitionNode_0 = AnimStateTransitionNode()
   AnimStateTransitionNode_2 = AnimStateTransitionNode()
   AnimStateTransitionNode_29 = AnimStateTransitionNode()
   AnimStateTransitionNode_31 = AnimStateTransitionNode()
  ]
  links [
   {
    dst=East
    rule=AnimGraph.AnimGraphNode_StateMachine_0.New State Machine.AnimStateNode_0.GroundLocomotion.AnimGraphNode_StateMachine_0.GroundLocomotion.AnimStateNode_1.Run.AnimGraphNode_StateMachine_0.New State Machine.AnimStateTransitionNode_2.Transition
    src=Conduit
    via=AnimStateTransitionNode_2
   }
   {
    dst=North
    rule=AnimGraph.AnimGraphNode_StateMachine_0.New State Machine.AnimStateNode_0.GroundLocomotion.AnimGraphNode_StateMachine_0.GroundLocomotion.AnimStateNode_1.Run.AnimGraphNode_StateMachine_0.New State Machine.AnimStateTransitionNode_29.AnimationTransitionGraph_2
    src=Conduit
    via=AnimStateTransitionNode_29
   }
   {
    dst=South
    rule=AnimGraph.AnimGraphNode_StateMachine_0.New State Machine.AnimStateNode_0.GroundLocomotion.AnimGraphNode_StateMachine_0.GroundLocomotion.AnimStateNode_1.Run.AnimGraphNode_StateMachine_0.New State Machine.AnimStateTransitionNode_31.AnimationTransitionGraph_2
    src=Conduit
    via=AnimStateTransitionNode_31
   }
   {
    dst=West
    rule=AnimGraph.AnimGraphNode_StateMachine_0.New State Machine.AnimStateNode_0.GroundLocomotion.AnimGraphNode_StateMachine_0.GroundLocomotion.AnimStateNode_1.Run.AnimGraphNode_StateMachine_0.New State Machine.AnimStateTransitionNode_0.Transition
    src=Conduit
    via=AnimStateTransitionNode_0
   }
  ]
 }
 Conduit {
  path=AnimGraph.AnimGraphNode_StateMachine_0.New State Machine.AnimStateNode_0.GroundLocomotion.AnimGraphNode_StateMachine_0.GroundLocomotion.AnimStateNode_1.Run.AnimGraphNode_StateMachine_0.New State Machine.AnimStateConduitNode_0.Conduit
  nodes [Result(bCanEnterTransition=true)]
 }
 North {
  path=AnimGraph.AnimGraphNode_StateMachine_0.New State Machine.AnimStateNode_0.GroundLocomotion.AnimGraphNode_StateMachine_0.GroundLocomotion.AnimStateNode_1.Run.AnimGraphNode_StateMachine_0.New State Machine.AnimStateNode_1.North
  nodes [
   n1 = Sequence Player 'FoldupStart_F_MED_Wayfare_Pack'() {
    bLoopAnimation=false
    sequence=&FoldupStart_F_MED_Wayfare_Pack
   }
   Output Animation Pose(Result=@n1)
  ]
 }
 South {
  path=AnimGraph.AnimGraphNode_StateMachine_0.New State Machine.AnimStateNode_0.GroundLocomotion.AnimGraphNode_StateMachine_0.GroundLocomotion.AnimStateNode_1.Run.AnimGraphNode_StateMachine_0.New State Machine.AnimStateNode_2.South
  nodes [
   Sequence Player 'Unfold_F_MED_Wayfare_Pack'() {sequence=/Game/Accessories/FORT_Backpacks/Backpack_F_MED_Wayfare/WayfareAnimations/Unfold_F_MED_Wayfare_Pack}
   Output Animation Pose(Result=Sequence Player 'UnfoldSouthShort_F_MED_Wayfare_Pack'() {sequence=/BRCosmetics/Accessories/FORT_Backpacks/Backpack_F_MED_Wayfare/WayfareAnimations/UnfoldSouthShort_F_MED_Wayfare_Pack})
  ]
 }
 East {
  path=AnimGraph.AnimGraphNode_StateMachine_0.New State Machine.AnimStateNode_0.GroundLocomotion.AnimGraphNode_StateMachine_0.GroundLocomotion.AnimStateNode_1.Run.AnimGraphNode_StateMachine_0.New State Machine.AnimStateNode_3.East
  nodes [Output Animation Pose(Result=Sequence Player 'FoldupStart_F_MED_Wayfare_Pack'() {sequence=&FoldupStart_F_MED_Wayfare_Pack})]
 }
 West {
  path=AnimGraph.AnimGraphNode_StateMachine_0.New State Machine.AnimStateNode_0.GroundLocomotion.AnimGraphNode_StateMachine_0.GroundLocomotion.AnimStateNode_1.Run.AnimGraphNode_StateMachine_0.New State Machine.AnimStateNode_4.West
  nodes [Output Animation Pose(Result=Sequence Player 'FoldupStart_F_MED_Wayfare_Pack'() {sequence=&FoldupStart_F_MED_Wayfare_Pack})]
 }
 Transition {
  path=AnimGraph.AnimGraphNode_StateMachine_0.New State Machine.AnimStateNode_0.GroundLocomotion.AnimGraphNode_StateMachine_0.GroundLocomotion.AnimStateNode_1.Run.AnimGraphNode_StateMachine_0.New State Machine.AnimStateTransitionNode_0.Transition
  nodes [
   n1 = Get As Fortnite M Avg Player Anim Blueprint() -> (As Fortnite M Avg Player Anim Blueprint:Fortnite M Avg Player Anim Blueprint Object Reference)
   Result(bCanEnterTransition="Equal (Enum)"(
     A:EFortCardinalDirection Enum=Get LocomotionCardinalDirection!(self:Fort Player Anim Instance Object Reference=@n1) -> (LocomotionCardinalDirection:EFortCardinalDirection Enum=North)
     B:EFortCardinalDirection Enum=West
    ))
  ]
 }
 Transition {
  path=AnimGraph.AnimGraphNode_StateMachine_0.New State Machine.AnimStateNode_0.GroundLocomotion.AnimGraphNode_StateMachine_0.GroundLocomotion.AnimStateNode_1.Run.AnimGraphNode_StateMachine_0.New State Machine.AnimStateTransitionNode_2.Transition
  nodes [
   n1 = Get As Fortnite M Avg Player Anim Blueprint()
   Result(bCanEnterTransition="Equal (Enum)"(A=Get LocomotionCardinalDirection(self=@n1) -> (LocomotionCardinalDirection=North), B=East))
  ]
 }
 AnimationTransitionGraph_2 {
  path=AnimGraph.AnimGraphNode_StateMachine_0.New State Machine.AnimStateNode_0.GroundLocomotion.AnimGraphNode_StateMachine_0.GroundLocomotion.AnimStateNode_1.Run.AnimGraphNode_StateMachine_0.New State Machine.AnimStateTransitionNode_29.AnimationTransitionGraph_2
  nodes [
   n1 = Get As Fortnite M Avg Player Anim Blueprint()
   Result(bCanEnterTransition="Equal (Enum)"(A=Get LocomotionCardinalDirection(self=@n1) -> (LocomotionCardinalDirection=North), B=North))
  ]
 }
 AnimationTransitionGraph_2 {
  path=AnimGraph.AnimGraphNode_StateMachine_0.New State Machine.AnimStateNode_0.GroundLocomotion.AnimGraphNode_StateMachine_0.GroundLocomotion.AnimStateNode_1.Run.AnimGraphNode_StateMachine_0.New State Machine.AnimStateTransitionNode_31.AnimationTransitionGraph_2
  nodes [
   n1 = Get As Fortnite M Avg Player Anim Blueprint()
   Result(bCanEnterTransition="Equal (Enum)"(A=Get LocomotionCardinalDirection(self=@n1) -> (LocomotionCardinalDirection=North), B=South))
  ]
 }
 Stop {
  path=AnimGraph.AnimGraphNode_StateMachine_0.New State Machine.AnimStateNode_0.GroundLocomotion.AnimGraphNode_StateMachine_0.GroundLocomotion.AnimStateNode_3.Stop
  nodes [
   Sequence Player 'Unfold_F_MED_Wayfare_Pack'() {bLoopAnimation=false, sequence=&Unfold_F_MED_Wayfare_Pack}
   Output Animation Pose(Result=Sequence Player 'UnfoldSouth_F_MED_Wayfare_Pack'() {sequence=/BRCosmetics/Accessories/FORT_Backpacks/Backpack_F_MED_Wayfare/WayfareAnimations/UnfoldSouth_F_MED_Wayfare_Pack})
  ]
 }
 Loop {
  path=AnimGraph.AnimGraphNode_StateMachine_0.New State Machine.AnimStateNode_0.GroundLocomotion.AnimGraphNode_StateMachine_0.GroundLocomotion.AnimStateNode_5.Loop
  nodes [Output Animation Pose(Result=Sequence Player 'FoldupLoop_F_MED_Wayfare_Pack'() {sequence=/BRCosmetics/Accessories/FORT_Backpacks/Backpack_F_MED_Wayfare/WayfareAnimations/FoldupLoop_F_MED_Wayfare_Pack})]
 }
 AnimationTransitionGraph_4 {
  path=AnimGraph.AnimGraphNode_StateMachine_0.New State Machine.AnimStateNode_0.GroundLocomotion.AnimGraphNode_StateMachine_0.GroundLocomotion.AnimStateTransitionNode_0.AnimationTransitionGraph_4
  nodes [Result(bCanEnterTransition=Get bIsMoving2D!(self:Fort Player Anim Instance Object Reference=Get As Fortnite M Avg Player Anim Blueprint()) -> (bIsMoving2D))]
 }
 AnimationTransitionGraph_4 {
  path=AnimGraph.AnimGraphNode_StateMachine_0.New State Machine.AnimStateNode_0.GroundLocomotion.AnimGraphNode_StateMachine_0.GroundLocomotion.AnimStateTransitionNode_1.AnimationTransitionGraph_4
  nodes [Result(bCanEnterTransition=NOT Boolean(A=Get bIsMoving2D!() -> (bIsMoving2D)))]
 }
 AnimationTransitionGraph_4 {
  path=AnimGraph.AnimGraphNode_StateMachine_0.New State Machine.AnimStateNode_0.GroundLocomotion.AnimGraphNode_StateMachine_0.GroundLocomotion.AnimStateTransitionNode_11.AnimationTransitionGraph_4
  nodes [Result(bCanEnterTransition=Get bIsDBNO())]
 }
 AnimationTransitionGraph_4 {
  path=AnimGraph.AnimGraphNode_StateMachine_0.New State Machine.AnimStateNode_0.GroundLocomotion.AnimGraphNode_StateMachine_0.GroundLocomotion.AnimStateTransitionNode_2.AnimationTransitionGraph_4
  nodes [Result(bCanEnterTransition=true)]
 }
 AnimationTransitionGraph_4 {
  path=AnimGraph.AnimGraphNode_StateMachine_0.New State Machine.AnimStateNode_0.GroundLocomotion.AnimGraphNode_StateMachine_0.GroundLocomotion.AnimStateTransitionNode_4.AnimationTransitionGraph_4
  nodes [Result(bCanEnterTransition=true)]
 }
 Targeting {
  path=AnimGraph.AnimGraphNode_StateMachine_0.New State Machine.AnimStateNode_1.Targeting
  nodes [
   Sequence Player 'FoldupStart_F_MED_Wayfare_Pack'() {sequence=&FoldupStart_F_MED_Wayfare_Pack}
   Output Animation Pose(Result=Sequence Player 'Targeting_F_MED_Wayfare_Pack'() {sequence=/BRCosmetics/Accessories/FORT_Backpacks/Backpack_F_MED_Wayfare/WayfareAnimations/Targeting_F_MED_Wayfare_Pack})
  ]
 }
 Melee {
  path=AnimGraph.AnimGraphNode_StateMachine_0.New State Machine.AnimStateNode_10.Melee
  nodes [
   Sequence Player 'melee_F_MED_Wayfare_Pack'() {sequence=/BRCosmetics/Accessories/FORT_Backpacks/Backpack_F_MED_Wayfare/WayfareAnimations/melee_F_MED_Wayfare_Pack}
   AnimGraphNode_StateMachine_0 = New State Machine()
   Output Animation Pose(Result=@AnimGraphNode_StateMachine_0)
  ]
 }
 New State Machine {
  entry=meleestart
  path=AnimGraph.AnimGraphNode_StateMachine_0.New State Machine.AnimStateNode_10.Melee.AnimGraphNode_StateMachine_0.New State Machine
  ids=id
  nodes [
   AnimStateNode_1 = meleestart() {is=title AnimStateNode}
   AnimStateNode_2 = meleeloop() {is=title AnimStateNode}
   AnimStateTransitionNode_0 = AnimStateTransitionNode() {bAutomaticRuleBasedOnSequencePlayerInState=true}
  ]
  links [{
    dst=meleeloop
    rule=AnimGraph.AnimGraphNode_StateMachine_0.New State Machine.AnimStateNode_10.Melee.AnimGraphNode_StateMachine_0.New State Machine.AnimStateTransitionNode_0.Transition
    src=meleestart
    via=AnimStateTransitionNode_0
   }]
 }
 meleestart {
  path=AnimGraph.AnimGraphNode_StateMachine_0.New State Machine.AnimStateNode_10.Melee.AnimGraphNode_StateMachine_0.New State Machine.AnimStateNode_1.meleestart
  nodes [Output Animation Pose(Result=Sequence Player 'melee_F_MED_Wayfare_Pack'() {sequence=&melee_F_MED_Wayfare_Pack})]
 }
 meleeloop {
  path=AnimGraph.AnimGraphNode_StateMachine_0.New State Machine.AnimStateNode_10.Melee.AnimGraphNode_StateMachine_0.New State Machine.AnimStateNode_2.meleeloop
  nodes [Output Animation Pose(Result=Sequence Player 'meleeloop_F_MED_Wayfare_Pack'() {sequence=/BRCosmetics/Accessories/FORT_Backpacks/Backpack_F_MED_Wayfare/WayfareAnimations/meleeloop_F_MED_Wayfare_Pack})]
 }
 Transition {
  path=AnimGraph.AnimGraphNode_StateMachine_0.New State Machine.AnimStateNode_10.Melee.AnimGraphNode_StateMachine_0.New State Machine.AnimStateTransitionNode_0.Transition
  nodes [Result(bCanEnterTransition=true)]
 }
 DBNO {
  path=AnimGraph.AnimGraphNode_StateMachine_0.New State Machine.AnimStateNode_2.DBNO
  nodes [
   Sequence Player 'DBNOexplode_F_MED_Wayfare_Pack_Part'() {bLoopAnimation=false, sequence=/Game/Accessories/FORT_Backpacks/Backpack_F_MED_Wayfare/WayfareAnimations/DBNOexplode_F_MED_Wayfare_Pack_Part}
   Output Animation Pose(Result=Sequence Player 'DBNO_F_MED_Wayfare_Pack'() {sequence=/BRCosmetics/Accessories/FORT_Backpacks/Backpack_F_MED_Wayfare/WayfareAnimations/DBNO_F_MED_Wayfare_Pack})
  ]
 }
 SkyDiveMode {
  path=AnimGraph.AnimGraphNode_StateMachine_0.New State Machine.AnimStateNode_3.SkyDiveMode
  nodes [
   AnimGraphNode_StateMachine_0 = Skydiving State() {is=title AnimGraphNode_StateMachine}
   Output Animation Pose(Result=@AnimGraphNode_StateMachine_0)
  ]
 }
 Skydiving State {
  entry=Parachuting
  path=AnimGraph.AnimGraphNode_StateMachine_0.New State Machine.AnimStateNode_3.SkyDiveMode.AnimGraphNode_StateMachine_0.Skydiving State
  ids=id
  nodes [
   AnimStateNode_0 = Parachuting() {is=title AnimStateNode}
   AnimStateNode_1 = Diving() {is=title AnimStateNode}
   AnimStateNode_2 = Floating() {is=title AnimStateNode}
   AnimStateTransitionNode_13 = AnimStateTransitionNode()
   AnimStateTransitionNode_14 = AnimStateTransitionNode()
   AnimStateTransitionNode_15 = AnimStateTransitionNode()
   AnimStateTransitionNode_16 = AnimStateTransitionNode()
   AnimStateTransitionNode_17 = AnimStateTransitionNode()
   AnimStateTransitionNode_18 = AnimStateTransitionNode()
  ]
  links [
   {
    dst=Floating
    rule=AnimGraph.AnimGraphNode_StateMachine_0.New State Machine.AnimStateNode_3.SkyDiveMode.AnimGraphNode_StateMachine_0.Skydiving State.AnimStateTransitionNode_16.Transition
    src=Diving
    via=AnimStateTransitionNode_16
   }
   {
    dst=Parachuting
    rule=AnimGraph.AnimGraphNode_StateMachine_0.New State Machine.AnimStateNode_3.SkyDiveMode.AnimGraphNode_StateMachine_0.Skydiving State.AnimStateTransitionNode_18.Transition
    src=Diving
    via=AnimStateTransitionNode_18
   }
   {
    dst=Diving
    rule=AnimGraph.AnimGraphNode_StateMachine_0.New State Machine.AnimStateNode_3.SkyDiveMode.AnimGraphNode_StateMachine_0.Skydiving State.AnimStateTransitionNode_15.Transition
    src=Floating
    via=AnimStateTransitionNode_15
   }
   {
    dst=Parachuting
    rule=AnimGraph.AnimGraphNode_StateMachine_0.New State Machine.AnimStateNode_3.SkyDiveMode.AnimGraphNode_StateMachine_0.Skydiving State.AnimStateTransitionNode_17.Transition
    src=Floating
    via=AnimStateTransitionNode_17
   }
   {
    dst=Diving
    rule=AnimGraph.AnimGraphNode_StateMachine_0.New State Machine.AnimStateNode_3.SkyDiveMode.AnimGraphNode_StateMachine_0.Skydiving State.AnimStateTransitionNode_13.Transition
    src=Parachuting
    via=AnimStateTransitionNode_13
   }
   {
    dst=Floating
    rule=AnimGraph.AnimGraphNode_StateMachine_0.New State Machine.AnimStateNode_3.SkyDiveMode.AnimGraphNode_StateMachine_0.Skydiving State.AnimStateTransitionNode_14.Transition
    src=Parachuting
    via=AnimStateTransitionNode_14
   }
  ]
 }
 Parachuting {
  path=AnimGraph.AnimGraphNode_StateMachine_0.New State Machine.AnimStateNode_3.SkyDiveMode.AnimGraphNode_StateMachine_0.Skydiving State.AnimStateNode_0.Parachuting
  nodes [
   Apply Additive(
    Base=Sequence Player 'FoldupStart_F_MED_Wayfare_Pack'() {sequence=&FoldupStart_F_MED_Wayfare_Pack}
    Additive=Blendspace Player 'BS_Binary_Twin_Leans_Parachuting'(X=0, Y=0) {blendSpace=/BRCosmetics/Accessories/FORT_Backpacks/Backpack_F_MED_Binary_Twin/Animation/BS_Binary_Twin_Leans_Parachuting}
   ) {alphaBoolBlend {blendOption=Linear}}
   Binary_Twin_Movement_Poses_PoseAsset : Float!(PoseWeight:float=1) -> (Pose:Pose Link Structure) {
    is=title AnimGraphNode_PoseByName
    poseAsset=/BRCosmetics/Accessories/FORT_Backpacks/Backpack_F_MED_Binary_Twin/Animation/Binary_Twin_Movement_Poses_PoseAsset
    poseName=Float
   }
   Output Animation Pose(Result=Sequence Player 'Idle_F_MED_Wayfare_Pack'() {sequence=&Idle_F_MED_Wayfare_Pack})
  ]
 }
 Diving {
  path=AnimGraph.AnimGraphNode_StateMachine_0.New State Machine.AnimStateNode_3.SkyDiveMode.AnimGraphNode_StateMachine_0.Skydiving State.AnimStateNode_1.Diving
  nodes [
   n1 = Break Fort Anim Input Skydiving!(FortAnimInput_Skydiving:"Fort Anim Input Skydiving Structure (by ref)"=Get Skydiving!(self:Fort Player Anim Instance Object Reference=Get As Fortnite M Avg Player Anim Blueprint()) -> (Skydiving:Fort Anim Input Skydiving Structure)) -> (
    ParaGilderRootOffset:Vector=~
    ParaGliderRootRotationOverride:Rotator=~
    bUseParaGlideRootModifier:bool=~
    bIsSkydivingFromLaunchPad:bool=~
    bIsSkydivingFromBus:bool=~
    bIsInVortex:bool=~
    bIsDiving:bool=~
    bIsParachuteOpen:bool=~
    bIsSkydiving:bool=~
    bIsParachuteLeaning:bool=~
    bIsSkydiveLeaning:bool=~
    bIsLeaning:bool=~
    bIsSkydiveDiveMode:bool=~
    bPlaySkydiveDrift:bool=~
    bSkydiveDriftDelayActive:bool=~
    bSkydiveDriftAnimAllowed:bool=~
    bIsGliderRight:bool=~
    bIsGliderCenter:bool=~
    bIsGliderLeft:bool=~
    bIsGliderForward:bool=~
    bIsGliderBack:bool=~
    bHasDeployedAndIsStandingUpright:bool=~
    LocalAccelForward:float=~
    LocalAccelRight:float=~
    LocalVelocityRight:float=~
    SkydiveAimPitch:float=~
    SkydiveAimYaw:float=~
    DeployChuteAnimRate:float=~
    SkydiveDriftAnimRate:float=~
    SkydiveDriftAnimRateCurrent:float=~
    SkydiveFidgetAnimRate:float=~
    SkydiveFidgetAnimRateCurrent:float=~
    SkydiveAdditiveAlpha:float=~
    SkydiveDriftDelay:float=~
    SkydiveDriftAnim:int
    SkydiveDriftAnimMax:int
    LaunchpadAnim:int
    LocalAccelDir:ESkydivingDirection Enum
    DirectionLast:ESkydivingDirection Enum
    PoseAdditive:Anim Sequence Object Reference
   ) {is=title K2Node_BreakStruct}
   Output Animation Pose(Result=Apply Additive(
     Base=Sequence Player 'Dive_F_MED_Wayfare_Pack'() {sequence=/BRCosmetics/Accessories/FORT_Backpacks/Backpack_F_MED_Wayfare/WayfareAnimations/Dive_F_MED_Wayfare_Pack}
     Additive=Blendspace Player 'BS_F_MED_Wayfare_Pack'(X=@n1.LocalAccelRight, Y=@n1.LocalAccelForward) {blendSpace=/BRCosmetics/Accessories/FORT_Backpacks/Backpack_F_MED_Wayfare/WayfareAnimations/BS_F_MED_Wayfare_Pack}
    ) {alphaBoolBlend {blendOption=Linear}})
  ]
 }
 Floating {
  path=AnimGraph.AnimGraphNode_StateMachine_0.New State Machine.AnimStateNode_3.SkyDiveMode.AnimGraphNode_StateMachine_0.Skydiving State.AnimStateNode_2.Floating
  nodes [
   n1 = Binary_Twin_Movement_Poses_PoseAsset : Float(PoseWeight=1) {poseAsset=&Binary_Twin_Movement_Poses_PoseAsset, poseName=Float}
   Apply Additive(
    Base=@n1
    Additive=Blendspace Player 'BS_Binary_Twin_Leans_Floating'(X=0, Y=0) {blendSpace=/BRCosmetics/Accessories/FORT_Backpacks/Backpack_F_MED_Binary_Twin/Animation/BS_Binary_Twin_Leans_Floating}
   ) {alphaBoolBlend {blendOption=Linear}}
   Output Animation Pose(Result=Sequence Player 'FoldupLoop_F_MED_Wayfare_Pack'() {sequence=&FoldupLoop_F_MED_Wayfare_Pack})
  ]
 }
 Transition {
  path=AnimGraph.AnimGraphNode_StateMachine_0.New State Machine.AnimStateNode_3.SkyDiveMode.AnimGraphNode_StateMachine_0.Skydiving State.AnimStateTransitionNode_13.Transition
  nodes [Result(bCanEnterTransition=NOT Boolean(A=Get bIsParachuteOpened()))]
 }
 Transition {
  path=AnimGraph.AnimGraphNode_StateMachine_0.New State Machine.AnimStateNode_3.SkyDiveMode.AnimGraphNode_StateMachine_0.Skydiving State.AnimStateTransitionNode_14.Transition
  nodes [Result(bCanEnterTransition=NOT Boolean(A=Get bIsParachuteOpened()))]
 }
 Transition {
  path=AnimGraph.AnimGraphNode_StateMachine_0.New State Machine.AnimStateNode_3.SkyDiveMode.AnimGraphNode_StateMachine_0.Skydiving State.AnimStateTransitionNode_15.Transition
  nodes [Result(bCanEnterTransition=Get bIsSkydiveDiveMode())]
 }
 Transition {
  path=AnimGraph.AnimGraphNode_StateMachine_0.New State Machine.AnimStateNode_3.SkyDiveMode.AnimGraphNode_StateMachine_0.Skydiving State.AnimStateTransitionNode_16.Transition
  nodes [Result(bCanEnterTransition=NOT Boolean(A=Get bIsSkydiveDiveMode()))]
 }
 Transition {
  path=AnimGraph.AnimGraphNode_StateMachine_0.New State Machine.AnimStateNode_3.SkyDiveMode.AnimGraphNode_StateMachine_0.Skydiving State.AnimStateTransitionNode_17.Transition
  nodes [Result(bCanEnterTransition=Get bIsParachuteOpened())]
 }
 Transition {
  path=AnimGraph.AnimGraphNode_StateMachine_0.New State Machine.AnimStateNode_3.SkyDiveMode.AnimGraphNode_StateMachine_0.Skydiving State.AnimStateTransitionNode_18.Transition
  nodes [Result(bCanEnterTransition=Get bIsParachuteOpened())]
 }
 Emote {
  path=AnimGraph.AnimGraphNode_StateMachine_0.New State Machine.AnimStateNode_4.Emote
  nodes [
   Sequence Player 'melee_F_MED_Wayfare_Pack'() {sequence=&melee_F_MED_Wayfare_Pack}
   AnimGraphNode_StateMachine_0 = New State Machine()
   Output Animation Pose(Result=@AnimGraphNode_StateMachine_0)
  ]
 }
 New State Machine {
  entry=emotestart
  path=AnimGraph.AnimGraphNode_StateMachine_0.New State Machine.AnimStateNode_4.Emote.AnimGraphNode_StateMachine_0.New State Machine
  ids=id
  nodes [
   AnimStateNode_1 = emotestart() {is=title AnimStateNode}
   AnimStateNode_2 = emoteloop() {is=title AnimStateNode}
   AnimStateTransitionNode_0 = AnimStateTransitionNode() {bAutomaticRuleBasedOnSequencePlayerInState=true}
  ]
  links [{
    dst=emoteloop
    rule=AnimGraph.AnimGraphNode_StateMachine_0.New State Machine.AnimStateNode_4.Emote.AnimGraphNode_StateMachine_0.New State Machine.AnimStateTransitionNode_0.Transition
    src=emotestart
    via=AnimStateTransitionNode_0
   }]
 }
 emotestart {
  path=AnimGraph.AnimGraphNode_StateMachine_0.New State Machine.AnimStateNode_4.Emote.AnimGraphNode_StateMachine_0.New State Machine.AnimStateNode_1.emotestart
  nodes [Output Animation Pose(Result=Sequence Player 'melee_F_MED_Wayfare_Pack'() {sequence=&melee_F_MED_Wayfare_Pack})]
 }
 emoteloop {
  path=AnimGraph.AnimGraphNode_StateMachine_0.New State Machine.AnimStateNode_4.Emote.AnimGraphNode_StateMachine_0.New State Machine.AnimStateNode_2.emoteloop
  nodes [Output Animation Pose(Result=Sequence Player 'meleeloop_F_MED_Wayfare_Pack'() {sequence=&meleeloop_F_MED_Wayfare_Pack})]
 }
 Transition {
  path=AnimGraph.AnimGraphNode_StateMachine_0.New State Machine.AnimStateNode_4.Emote.AnimGraphNode_StateMachine_0.New State Machine.AnimStateTransitionNode_0.Transition
  nodes [Result(bCanEnterTransition=true)]
 }
 DBNO_End {
  path=AnimGraph.AnimGraphNode_StateMachine_0.New State Machine.AnimStateNode_5.DBNO_End
  nodes [
   Sequence Player 'FoldupStart_F_MED_Wayfare_Pack'() {sequence=&FoldupStart_F_MED_Wayfare_Pack}
   n1 = Sequence Player 'revive_F_MED_Wayfare_Pack'() {bLoopAnimation=false, sequence=/BRCosmetics/Accessories/FORT_Backpacks/Backpack_F_MED_Wayfare/WayfareAnimations/revive_F_MED_Wayfare_Pack}
   Output Animation Pose(Result=@n1)
  ]
 }
 Jump {
  path=AnimGraph.AnimGraphNode_StateMachine_0.New State Machine.AnimStateNode_7.Jump
  nodes [
   n1 = Sequence Player 'Ghost_Jump_F_MED_Wayfare_Pack'() {
    bLoopAnimation=false
    sequence=/BRCosmetics/Accessories/FORT_Backpacks/Backpack_F_MED_Wayfare/WayfareAnimations/Ghost_Jump_F_MED_Wayfare_Pack
   }
   Output Animation Pose(Result=@n1)
  ]
 }
 Sliding {
  path=AnimGraph.AnimGraphNode_StateMachine_0.New State Machine.AnimStateNode_9.Sliding
  nodes [Output Animation Pose(Result=Sequence Player 'Slide_F_MED_Wayfare_Pack'() {sequence=/BRCosmetics/Accessories/FORT_Backpacks/Backpack_F_MED_Wayfare/WayfareAnimations/Slide_F_MED_Wayfare_Pack})]
 }
 Transition {
  path=AnimGraph.AnimGraphNode_StateMachine_0.New State Machine.AnimStateTransitionNode_0.Transition
  nodes [Result(bCanEnterTransition=Get bIsPlayingMeleeAnim!(self:Fort Player Anim Instance Object Reference=Get As Fortnite M Avg Player Anim Blueprint()) -> (bIsPlayingMeleeAnim))]
 }
 Transition {
  path=AnimGraph.AnimGraphNode_StateMachine_0.New State Machine.AnimStateTransitionNode_1.Transition
  nodes [
   n1 = Get As Fortnite M Avg Player Anim Blueprint()
   Result(bCanEnterTransition=NOT Boolean(A=Get bIsPlayingMeleeAnim(self=@n1)))
  ]
 }
 AnimationTransitionGraph_2 {
  path=AnimGraph.AnimGraphNode_StateMachine_0.New State Machine.AnimStateTransitionNode_10.AnimationTransitionGraph_2
  nodes [Result(bCanEnterTransition=Get bIsDBNO())]
 }
 Transition {
  path=AnimGraph.AnimGraphNode_StateMachine_0.New State Machine.AnimStateTransitionNode_11.Transition
  nodes [Result(bCanEnterTransition=true)]
 }
 Transition {
  path=AnimGraph.AnimGraphNode_StateMachine_0.New State Machine.AnimStateTransitionNode_15.Transition
  nodes [Result(bCanEnterTransition=Get bIsTargeting())]
 }
 Transition {
  path=AnimGraph.AnimGraphNode_StateMachine_0.New State Machine.AnimStateTransitionNode_16.Transition
  nodes [Result(bCanEnterTransition=NOT Boolean(A=Get bIsTargeting()))]
 }
 Transition {
  path=AnimGraph.AnimGraphNode_StateMachine_0.New State Machine.AnimStateTransitionNode_17.Transition
  nodes [Result(bCanEnterTransition=Get bIsSliding())]
 }
 Transition {
  path=AnimGraph.AnimGraphNode_StateMachine_0.New State Machine.AnimStateTransitionNode_18.Transition
  nodes [Result(bCanEnterTransition=NOT Boolean(A=Get bIsSliding()))]
 }
 Transition {
  path=AnimGraph.AnimGraphNode_StateMachine_0.New State Machine.AnimStateTransitionNode_2.Transition
  nodes [Result(bCanEnterTransition=OR Boolean(A=Get bIsPlayingEmote(), B=Get bIsPlayingAnyEmote() -> (bIsPlayingAnyEmote:bool)))]
 }
 AnimationTransitionGraph_2 {
  path=AnimGraph.AnimGraphNode_StateMachine_0.New State Machine.AnimStateTransitionNode_3.AnimationTransitionGraph_2
  nodes [Result(bCanEnterTransition=NOT Boolean(A=Get bIsDBNO()))]
 }
 Transition {
  path=AnimGraph.AnimGraphNode_StateMachine_0.New State Machine.AnimStateTransitionNode_4.Transition
  nodes [Result(bCanEnterTransition=NOT Boolean(A=Get bIsDBNO()))]
 }
 Transition {
  path=AnimGraph.AnimGraphNode_StateMachine_0.New State Machine.AnimStateTransitionNode_5.Transition
  nodes [
   Result(bCanEnterTransition=Get bIsSkyDiving())
   Get As Fortnite M Avg Player Anim Blueprint()
   Get bIsSkydiving!(self:Fort Player Pawn Object Reference) -> (bIsSkydiving:bool)
  ]
 }
 Transition {
  path=AnimGraph.AnimGraphNode_StateMachine_0.New State Machine.AnimStateTransitionNode_6.Transition
  nodes [Result(bCanEnterTransition=Get bIsJumping())]
 }
 Transition {
  path=AnimGraph.AnimGraphNode_StateMachine_0.New State Machine.AnimStateTransitionNode_7.Transition
  nodes [Result(bCanEnterTransition=NOT Boolean(A=Get bIsSkyDiving()))]
 }
 Transition {
  path=AnimGraph.AnimGraphNode_StateMachine_0.New State Machine.AnimStateTransitionNode_8.Transition
  nodes [
   n1 = Get bIsPlayingEmote()
   n2 = Get bIsPlayingAnyEmote()
   Result(bCanEnterTransition=AND Boolean(A=NOT Boolean(A=@n2), B=NOT Boolean(A=@n1)))
  ]
 }
 Transition {
  path=AnimGraph.AnimGraphNode_StateMachine_0.New State Machine.AnimStateTransitionNode_9.Transition
  nodes [Result(bCanEnterTransition=NOT Boolean(A=Get bIsJumping()))]
 }
 EventGraph {nodes [
   Event BlueprintInitializeAnimation()
   > n1 = Cast To PlayerPawn_Athena(Object=TryGetPawnOwner())
   > n2 = Set As Player Pawn Athena(As Player Pawn Athena:Player Pawn Athena Object Reference=@n1) -> (Output_Get:Player Pawn Athena Object Reference)
   > n3 = Cast To Fortnite_M_Avg_Player_AnimBlueprint!(execute, Object=GetAnimInstance(self=Get Mesh(self=@n2))) -> (then, CastFailed, AsFortnite M Avg Player Anim Blueprint:Fortnite M Avg Player Anim Blueprint Object Reference) {is=title K2Node_DynamicCast}
   > Set As Fortnite M Avg Player Anim Blueprint(As Fortnite M Avg Player Anim Blueprint:Fortnite M Avg Player Anim Blueprint Object Reference=@n3) -> (Output_Get:Fortnite M Avg Player Anim Blueprint Object Reference)
   n4 = Get Skydiving(self=Get As Fortnite M Avg Player Anim Blueprint())
   PrintString!(
    execute
    InString="To String (Float)"(InDouble=Break Fort Anim Input Skydiving(FortAnimInput_Skydiving=@n4) -> (
      ParaGilderRootOffset=~
      ParaGliderRootRotationOverride=~
      bUseParaGlideRootModifier=~
      bIsSkydivingFromLaunchPad=~
      bIsSkydivingFromBus=~
      bIsInVortex=~
      bIsDiving=~
      bIsParachuteOpen=~
      bIsSkydiving=~
      bIsParachuteLeaning=~
      bIsSkydiveLeaning=~
      bIsLeaning=~
      bIsSkydiveDiveMode=~
      bPlaySkydiveDrift=~
      bSkydiveDriftDelayActive=~
      bSkydiveDriftAnimAllowed=~
      bIsGliderRight=~
      bIsGliderCenter=~
      bIsGliderLeft=~
      bIsGliderForward=~
      bIsGliderBack=~
      bHasDeployedAndIsStandingUpright=~
      LocalAccelForward=~
      LocalAccelRight=~
      LocalVelocityRight=~
      SkydiveAimPitch=~
      SkydiveAimYaw=~
      DeployChuteAnimRate=~
      SkydiveDriftAnimRate=~
      SkydiveDriftAnimRateCurrent=~
      SkydiveFidgetAnimRate=~
      SkydiveFidgetAnimRateCurrent=~
      SkydiveAdditiveAlpha=~
      SkydiveDriftDelay=~
     ).LocalAccelRight)
    bPrintToScreen
    bPrintToLog
    TextColor
    Duration
    Key
   ) -> (then)
   Event BlueprintUpdateAnimation()
  ]}
]
variables [
 As Fortnite M Avg Player Anim Blueprint {type=Fortnite_M_Avg_Player_AnimBlueprint_C, default=null}
 As Player Pawn Athena {type=PlayerPawn_Athena_C}
 NewVar {type=boolean}
]
